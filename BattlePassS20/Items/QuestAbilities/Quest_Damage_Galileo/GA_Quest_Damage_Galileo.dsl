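; Battle pass S20 Galileo quest ability that credits damage dealt to enemy players carrying a Mythic item tagged Lobster.
; On the activating gameplay event, it calls the parent, then requires player pawns as instigator and target, with the target off the owner's team.
; It scans the target's Mythic inventory for the Lobster tag and reports the rounded event magnitude to the owner's quest manager as a custom stat.
; EndAbility is also called inside the item loop.
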(event Ability|EventActivateAbilityFromEvent
  (params (EventData :type Gameplay%20Event%20Data%20Structure%20%28by%20ref%29))
  (bind (_eventtag _instigator _target _optionalobject _optionalobject2 _contexthandle _instigatortags _targettags _eventmagnitude) (Utilities|Struct|BreakGameplayEventData EventData))
  ;[1 1 1] Check if target is a Fort Pawn that's carrying Lobster
  (|CallParentFunction EventData)
  (Utilities|Casting|CastToFortPlayerPawn _instigator
    (:then
      (bind _asfort_player_pawn (Utilities|Casting|CastToFortPlayerPawn _target)
        (:then
          (if (not (Team|OnSameTeam _asfort_player_pawn (Variables|GAVars|Data|GetOwningPlayerPawn)))
            (bind _items (InventoryManagement|GetItemsfromInventoryOwnerforRarity (Fortnite|GetFortPlayerControllerfromActor _asfort_player_pawn) "Mythic"))
            (bind _array_element (Utilities|Array|ForEachLoopwithBreak _items)
              (:LoopBody
                (if (GameplayTags|HasTag (GameplayTags|GetOwnedGameplayTags _array_element :!class /Script/GameplayTags.GameplayTagAssetInterface) (Variables|Default|GetT_Lobster))
                  (Quest|SendComplexCustomStatEvent (Variables|GAVars|Data|GetOwningPlayerQuestManager) 0 "(GameplayTags=)" (Variables|Default|GetTargetQuestTag) (Math|Float|Round _eventmagnitude)))
                (Ability|EndAbility))
              (:Completed
                _))
            (Ability|EndAbility)))
        (:CastFailed
          _)))
    (:CastFailed
      _)) ; Ensure it is a player doing the damage
  (Ability|EndAbility))
